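# Niagara swing trail for the EarPsych pickaxe, drawing short-lived animation-trail ribbons from socket pairs.
# One emitter spawns three anim trails, initializes their bone indices and feeds three ribbon renderers.
# Color curves blend from magenta-purple to green-cyan by User.Variant, with alpha scaled by time of day and System.TrailFade.
# Ribbons narrow over life, while secondary color, secondary width and an emitter-age-driven zigzag width are written as extra particle attributes.
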NiagaraSystem /BRCosmetics/Weapons/FORT_Melee/Pickaxe_EarPsych/FX/NS_Pickaxe_Earpsych_Trail
dependencies {
 usedDataInterfaces [&NiagaraDataInterfaceColorCurve, &NiagaraDataInterfaceCurve, &NiagaraDataInterfaceSocketReader]
 usedDynamicInputs [&MergeHelper_FloatSwitch]
 usedModules [
  &SystemState
  &self:SystemSpawnScript.SystemScriptSource.SystemScriptGraph.NiagaraNodeAssignment_1.SetVariables_#1
  &self:SystemSpawnScript.SystemScriptSource.SystemScriptGraph.NiagaraNodeAssignment_0.SetVariables_#2
  &self:SystemSpawnScript.SystemScriptSource.SystemScriptGraph.NiagaraNodeAssignment_3.SetVariables_#3
 ]
}
emitters {Trail {
  emitterUpdateScript [
   &EmitterState()
   &FN_SpawnAnimTrail(AnimTrailWidthMode=::FromSecond, Number of Anim Trails=3)
  ]
  particleSpawnScript [
   &FN_InitializeAnimTrailBoneIndecies()
   &FN_InitializeAnimTrailBoneIndecies(AnimTrail Selection=1, First Bone or Socket Index=2, Second Bone or Socket Index=3) {module=FN_InitializeAnimTrailBoneIndecies001}
   &FN_InitializeAnimTrailBoneIndecies(AnimTrail Selection=2, First Bone or Socket Index=4, Second Bone or Socket Index=5) {module=FN_InitializeAnimTrailBoneIndecies003}
   &FN_InitializeAnimTrail(
    AnimTrailWidth=0.5
    Color=$System.TODColorMultiplier
    Lifetime=&RandomRangeFloat(Minimum=0.22, Maximum=0.22)
   ) {module=FN_InitializeAnimTrail001}
  ]
  particleUpdateScript [
   &FN_UpdateAnimTrail()
   &ScaleColor(
    Scale Alpha=&FloatFromCurve(
     FloatCurve=DataInterface() {propertyValues {Curve {
        defaultValue=1
        keys [
         {arriveTangent=-0.5885, interpMode=RCIM_Cubic, leaveTangent=-0.5885, tangentMode=RCTM_User, time=0, value=1}
         {interpMode=RCIM_Cubic, tangentMode=RCTM_User, time=1, value=0}
        ]
       }}}
     Scale Curve=&Multiply_Float(
      A=$System.TrailFade
      B=&FN_Select_Float_from_Time_Of_Day(Morning=0.2, Day=0.4, Evening=0.2, Night=0.1)
     )
    )
    Scale RGB=&MakeVectorFromLinearColorRGB(Color=&LerpLinearColors(
      StartColor=&ColorFromCurve(ColorCurve=DataInterface() {propertyValues {
         AlphaCurve {keys [<at 0.6041 1 RCIM_Cubic>,<at 1 0 RCIM_Cubic>]}
         BlueCurve {keys [
           <at 0 5 RCIM_Cubic>
           <at 0.1114 1.9773>
           {arriveTangent=-4.5888, interpMode=RCIM_Cubic, leaveTangent=-4.5888, time=0.2572, value=0.3775}
           <at 0.4208 0.5573>
          ]}
         GreenCurve {keys [
           <at 0 0.1284 RCIM_Cubic>
           <at 0.1114 0.0974>
           {arriveTangent=-0.2682, interpMode=RCIM_Cubic, leaveTangent=-0.2682, time=0.2572, value=0.0508}
           <at 0.4208 0.0144>
          ]}
         RedCurve {keys [
           <at 0 5 RCIM_Cubic>
           <at 0.1114 3.3332>
           {arriveTangent=-10.3543, interpMode=RCIM_Cubic, leaveTangent=-10.3543, time=0.2572, value=0.8333}
           <at 0.4208 0.1291>
          ]}
        }})
      EndColor=&ColorFromCurve(ColorCurve=DataInterface() {propertyValues {
         AlphaCurve {keys [<at 0 1 RCIM_Cubic>,<at 1 0 RCIM_Cubic>]}
         BlueCurve {keys [<at 0 3.7521 RCIM_Cubic>, {arriveTangent=-6.1974, interpMode=RCIM_Cubic, leaveTangent=-6.1974, time=0.188, value=1}, <at 0.4441 1>]}
         GreenCurve {keys [<at 0 5 RCIM_Cubic>, {arriveTangent=-10.2871, interpMode=RCIM_Cubic, leaveTangent=-10.2871, time=0.188, value=1}, <at 0.4441 0.4317>]}
         RedCurve {keys [<at 0 0.0086 RCIM_Cubic>, {arriveTangent=0.0389, interpMode=RCIM_Cubic, leaveTangent=0.0389, time=0.188, value=0.061}, <at 0.4441 0.0258>]}
        }})
      LerpFactor=$User.Variant
     ))
    ScaleA=true
    ScaleRGB=true
   ) {module=ScaleColor001}
   &ScaleColor(
    Scale Alpha=&FloatFromCurve(
     FloatCurve=DataInterface() {propertyValues {Curve {
        defaultValue=1
        keys [
         {arriveTangent=-0.5885, interpMode=RCIM_Cubic, leaveTangent=-0.5885, tangentMode=RCTM_User, time=0, value=1}
         {interpMode=RCIM_Cubic, tangentMode=RCTM_User, time=1, value=0}
        ]
       }}}
     Scale Curve=&Multiply_Float(
      A=$System.TrailFade
      B=&FN_Select_Float_from_Time_Of_Day(Morning=0.2, Day=0.4, Evening=0.2, Night=0.75)
     )
    )
    Scale RGB=&MakeVectorFromLinearColorRGB(Color=&LerpLinearColors(
      StartColor=&ColorFromCurve(ColorCurve=DataInterface() {propertyValues {
         AlphaCurve {keys [<at 0.7401 1 RCIM_Cubic>,<at 1 0 RCIM_Cubic>]}
         BlueCurve {keys [
           <at 0 0.55 RCIM_Cubic>
           <at 0.1114 0.3712>
           {arriveTangent=-0.5096, interpMode=RCIM_Cubic, leaveTangent=-0.5096, time=0.2572, value=0.2346}
           <at 0.4208 0.2135>
          ]}
         GreenCurve {keys [<at 0 0.0005 RCIM_Cubic>,<at 0.1114 0>,<at 0.2572 0 RCIM_Cubic>,<at 0.4208 0>]}
         RedCurve {keys [
           <at 0 0.4585 RCIM_Cubic>
           <at 0.1114 0.2705>
           {arriveTangent=-0.6124, interpMode=RCIM_Cubic, leaveTangent=-0.6124, time=0.2572, value=0.159}
           <at 0.4208 0.081>
          ]}
        }})
      EndColor=&ColorFromCurve(ColorCurve=DataInterface() {propertyValues {
         AlphaCurve {keys [<at 0 1 RCIM_Cubic>,<at 1 0 RCIM_Cubic>]}
         BlueCurve {keys [<at 0 3.7521 RCIM_Cubic>, {arriveTangent=-6.1974, interpMode=RCIM_Cubic, leaveTangent=-6.1974, time=0.188, value=1}, <at 0.4441 1>]}
         GreenCurve {keys [<at 0 5 RCIM_Cubic>, {arriveTangent=-10.2871, interpMode=RCIM_Cubic, leaveTangent=-10.2871, time=0.188, value=1}, <at 0.4441 0.4317>]}
         RedCurve {keys [<at 0 0.0086 RCIM_Cubic>, {arriveTangent=0.0389, interpMode=RCIM_Cubic, leaveTangent=0.0389, time=0.188, value=0.061}, <at 0.4441 0.0258>]}
        }})
      LerpFactor=$User.Variant
     ))
    ScaleA=true
    ScaleRGB=true
   ) {module=ScaleColor002}
   &ScaleRibbonWidth(Ribbon Width Scale=&FloatFromCurve(
     FloatCurve=DataInterface() {propertyValues {Curve {keys [{arriveTangent=-0.0042, interpMode=RCIM_Cubic, leaveTangent=-0.0042, tangentMode=RCTM_User, time=0, value=1}, <at 1 0>]}}}
     Scale Curve=0.8
    ))
   /SetVariables(
    Particles.SecondaryColor:LinearColor=&MultiplyLinearColor(
     Linear Color=&MultiplyLinearColor(
      Linear Color=&LerpLinearColors(StartColor=<rgba 0.0005 0 0.02 0.15>, EndColor=<rgba 0 0.01 0.02 0.15>, LerpFactor=$User.Variant)
      Scale Factor=&MakeVector4_FromLinearColor(LinearColor=$System.TODColorMultiplier)
     )
     Scale Factor=&MakeVector4FromVectorAndFloat(
      "Vector (RGB)"=&MakeVectorFromLinearColorRGB(Color=<rgba 1 1 1 1>)
      "Float (Alpha)"=&Multiply_Float(
       A=&MakeFloatFromLinearColor(LinearColor=$System.TODColorMultiplier, Channel:ENiagaraFloatFromLinearColorOptions=::NewEnumerator3)
       B=&FloatFromCurve(FloatCurve=DataInterface() {propertyValues {Curve {keys [<at 0 1>,<at 0.5 0>]}}}, Scale Curve=$System.TrailFade)
      )
     )
    )
    Particles.SecondaryWidth:NiagaraFloat=&Multiply_Float(A=$Particles.RibbonWidth, B=1.5)
   ) {isSetParameters=true}
   /SetVariables(Particles.Ribbon ZigZag width:NiagaraFloat=&FloatFromCurve(
     FloatCurve=DataInterface() {propertyValues {Curve {keys [
         <at 0.0033 0 RCIM_Cubic>
         {interpMode=RCIM_Cubic, tangentMode=RCTM_User, time=0.1951, value=1}
         {interpMode=RCIM_Cubic, tangentMode=RCTM_User, time=0.66, value=1}
         {arriveTangent=-0.7325, interpMode=RCIM_Cubic, leaveTangent=-0.7325, tangentMode=RCTM_User, time=0.9979, value=0.8353}
        ]}}}
     Scale Curve=&FloatFromCurve(
      FloatCurve=DataInterface() {propertyValues {Curve {keys [
          {interpMode=RCIM_Cubic, tangentMode=RCTM_User, time=0, value=0}
          {
           arriveTangent=-0.1296
           arriveTangentWeight=0.145
           interpMode=RCIM_Cubic
           leaveTangent=-0.1296
           leaveTangentWeight=0.145
           tangentMode=RCTM_User
           tangentWeightMode=RCTWM_WeightedBoth
           time=0.1971
           value=0.9647
          }
          {interpMode=RCIM_Cubic, tangentMode=RCTM_User, time=0.9942, value=1.4379}
         ]}}}
      CurveIndex=$Emitter.Age
      Scale Curve=10
     )
    )) {isSetParameters=true}
  ]
  rendererClasses [&NiagaraRibbonRendererProperties]
  renderers [
   {rendererClass=&NiagaraRibbonRendererProperties}
   {rendererClass=&NiagaraRibbonRendererProperties, rendererIndex=1}
   {rendererClass=&NiagaraRibbonRendererProperties, rendererIndex=2}
  ]
 }}
summary {
 systemName=NS_Pickaxe_Earpsych_Trail
 emitters [{emitterName=Trail, rendererClasses [&NiagaraRibbonRendererProperties]}]
 userVariables [
  User.Reactivity {type=NiagaraFloat, defaultValue=0, description=""}
  User.RibbonIDOffset {type=NiagaraInt32}
  User.SwingSectionIndex {type=NiagaraInt32}
  User.Variant {type=NiagaraFloat, defaultValue=0, description=""}
 ]
}
